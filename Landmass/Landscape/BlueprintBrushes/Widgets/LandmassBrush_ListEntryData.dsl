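; Landmass landscape brush list entry data Blueprint, used by the editor brush widgets to represent one item in a brush list.
; It declares a single custom Init event with no logic attached.
; The graph holds no other events, variables or processing, so it serves as a stub or data holder for list entries.
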(event Custom|Init)
