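; Landmass brush widget actor Blueprint that holds a preview landscape as a child actor component.
; The construction script gets the actor from PreviewLandscapeChild, casts it to Landmass_PreviewLandscape and stores it in the PreviewLandscape variable.
; BeginPlay, overlap and Tick events are empty, so the graph contains no runtime logic.
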(event EventBeginPlay)

(event Collision|EventActorBeginOverlap
  (params (OtherActor :type Actor%20Object%20Reference)))

(event EventTick
  (params (DeltaSeconds :type Float%20%28single-precision%29)))

(fn ConstructionScript
  (bind _aslandmass_preview_landscape (Utilities|Casting|CastToLandmass_PreviewLandscape (Variables|/Script/Engine.ChildActorComponent|GetChildActor (Variables|Default|GetPreviewLandscapeChild))))
  (Variables|Default|SetPreviewLandscape _aslandmass_preview_landscape))
