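# Unlit translucent material that visualizes Landmass landscape brush bounds, drawn with depth testing disabled.
# TilingAAGrid calls on both UV axes form the extents outline, and a SphereMask on the brush edge falloff from LandmassBrush_ShapeMask forms the shape line.
# Each gets its own color, and Brightness scales their sum.
# Line widths account for object scale and a camera distance term from pixel depth, field of view and view size.
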Material /Landmass/Landscape/BlueprintBrushes/Materials/Internal/BrushBounds
graph {
 properties {bDisableDepthTest=true, blendMode=BLEND_Translucent, pixelDepthOffsetMode=PDOM_Legacy, shadingModel=MSM_Unlit, usedShadingModels=Unlit}
 nodes [
  Add(A=NamedRerouteUsage(), B=NamedRerouteUsage())
  n1 = VectorParameter() {parameterName=ExtentsOutlineColor, defaultValue=<rgba 0.75 0.25 0 0>}
  n2 = NamedRerouteUsage()
  n3 = VectorParameter() {parameterName=ShapeColor, defaultValue=<rgba 0 0.5 0.5 0>}
  n4 = NamedRerouteUsage()
  Multiply(A=Add(A=Multiply(A=@n1.RGB, B=@n2), B=Multiply(A=@n3.RGB, B=@n4)), B=ScalarParameter() {defaultValue=1, parameterName=Brightness})
  n5 = ViewProperty() {property=MEVP_TanHalfFieldOfView}
  n6 = ComponentMask(None=ViewProperty() {property=MEVP_ViewSize}.Property) {g=true, r=true}
  NamedRerouteDeclaration(None=Multiply(A=PixelDepth().R, B=Divide(A=@n5.Property, B=@n6))) {name=CameraDist, nodeColor=<rgba 0 1 0.7188 1>}
  n7 = ComponentMask(None=TextureCoordinate()) {r=true}
  n8 = MaterialFunctionCall() {materialFunction=&ObjectScale}
  n9 = Multiply(A=Divide(A=ScalarParameter() {defaultValue=0.1, parameterName=LineWidth}, B=Divide(A=@n8.Scale X) {constB=128}), B=NamedRerouteUsage())
  n10 = ComponentMask(None=TextureCoordinate()) {g=true}
  n11 = MaterialFunctionCall() {materialFunction=&ObjectScale}
  n12 = Multiply(A=Divide(A=ScalarParameter() {defaultValue=0.1, parameterName=LineWidth}, B=Divide(A=@n11.Scale Y) {constB=128}), B=NamedRerouteUsage())
  NamedRerouteDeclaration(None=Max(
    A=MaterialFunctionCall!(ScalarUV=@n7, Split=@n9) {materialFunction=/Engine/EditorMaterials/TilingAAGrid}.Result
    B=MaterialFunctionCall(ScalarUV=@n10, Split=@n12) {materialFunction=&TilingAAGrid}.Result
   )) {name=Extents Outline, nodeColor=<rgba 0.5703 1 0 1>}
  n13 = Max(A=ScalarParameter() {defaultValue=0.1, parameterName=FalloffWidth}) {constB=0.05}
  n14 = MaterialFunctionCall!(Falloff Width Override=@n13, Falloff Offset Override) {materialFunction=/Landmass/Landscape/BlueprintBrushes/Materials/BrushMaterials/BasicFunctions/LandmassBrush_ShapeMask}
  n15 = Multiply(A=ScalarParameter() {defaultValue=0.1, parameterName=LineWidth}, B=NamedRerouteUsage())
  n16 = Divide(
   A=Multiply(A=Divide(A=@n15, B=@n13), B=ScalarParameter() {defaultValue=64, parameterName=ShapeMaskLineAdjust})
   B=MaterialFunctionCall() {materialFunction=&ObjectScale}.Scale X
  )
  n17 = Multiply(A=@n16) {constB=2}
  n18 = ScalarParameter() {defaultValue=0.1, parameterName=FalloffWidth}
  n19 = ScalarParameter() {defaultValue=1, parameterName=ShowShapeMask}
  n20 = Constant()
  NamedRerouteDeclaration(None=Multiply(A=SphereMask(A=@n14.Brush Edge Falloff, B=@n17, Radius=@n16), B=If(A=@n18, "A > B"=@n19, "A == B"=@n20, "A < B"=@n20) {constB=0.0001})) {name=Shape Mask, nodeColor=<rgba 1 0 0.7969 1>}
  ScalarParameter() {defaultValue=0.1, parameterName=FalloffWidth}
 ]
}
