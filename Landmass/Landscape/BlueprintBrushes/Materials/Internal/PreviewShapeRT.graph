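# Internal unlit additive material that previews a Landmass blueprint brush's shape render target over the landscape, with depth testing disabled.
# World XY is remapped to landscape UVs using location and world size from MPC_Landscape.
# A parameter blends the saturated RT color with a DistanceField gradient, centered at 0.5 and scaled by GradientWidth, and a faint grid texture is added.
# Several other nodes, including two material function calls, are unconnected.
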Material /Landmass/Landscape/BlueprintBrushes/Materials/Internal/PreviewShapeRT
graph {
 properties {
  bAllowNegativeEmissiveColor=true
  bDisableDepthTest=true
  blendMode=BLEND_Additive
  pixelDepthOffsetMode=PDOM_Legacy
  shadingModel=MSM_Unlit
  usedShadingModels=Unlit
 }
 nodes [
  n1 = WorldPosition()
  n2 = CollectionParameter() {collection=/Landmass/Landscape/BlueprintBrushes/MPC/MPC_Landscape, parameterName=LandscapeLocation}
  n3 = Divide(
   A=Subtract(A=ComponentMask(None=@n1.XYZ) {g=true, r=true}, B=ComponentMask(None=@n2) {g=true, r=true})
   B=AppendVector(
    A=CollectionParameter() {collection=&MPC_Landscape, parameterName=WorldSizeX}
    B=CollectionParameter() {collection=&MPC_Landscape, parameterName=WorldSizeY}
   )
  )
  n4 = TextureSampleParameter2D(UVs=@n3) {parameterName=RT, texture=&Black}
  n5 = Subtract(A=TextureSampleParameter2D(UVs=@n3) {parameterName=DistanceField, texture=&Black}.R) {constB=0.5}
  n6 = ScalarParameter() {defaultValue=0.05, parameterName=GradientWidth}
  n7 = Divide(A=@n5, B=@n6)
  Add(
   A=LinearInterpolate(A=Saturate(None=@n4.RGB), B=Saturate(None=@n7), Alpha=ScalarParameter() {parameterName=Show Distance Field})
   B=Multiply(A=TextureSample(UVs=@n3) {texture=/Engine/EngineDebugMaterials/T_1x1_Grid}.R) {constB=0.25}
  )
  Add(A=@n5, B=Multiply(A=@n6) {constB=0.5})
  MaterialFunctionCall() {materialFunction=/Landmass/Landscape/BlueprintBrushes/MF/PointSampleLandscapeUVs}
  MaterialFunctionCall() {materialFunction=&WorldPositionBehindTranslucency}
 ]
}
