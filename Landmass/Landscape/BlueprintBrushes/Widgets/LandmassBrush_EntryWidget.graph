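# List-entry widget for the Landmass editor, showing one landscape brush actor by display name with up and down reorder buttons.
# When its list item is set, it stores the actor and casts through Landmass Actor and its brush manager to cache the editor widget.
# The buttons send MoveUp and MoveDown messages with the brush to that editor widget.
# Selection and hover recolor the background; expansion changes only print a debug string.
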WidgetBlueprint /Landmass/Landscape/BlueprintBrushes/Widgets/LandmassBrush_EntryWidget
graphs [
 EventGraph {nodes [
   "On Clicked (Button_Up)"() {is=title K2Node_ComponentBoundEvent}
   > MoveUp!(
    execute:Exec
    self:Object Reference=Get EditorWidget() -> (EditorWidget:EUW Landmass Editor Object Reference)
    Brush:Actor Object Reference=Get MyBrush() -> (MyBrush:Actor Object Reference)
   ) -> (then:Exec) {is=title K2Node_Message}
   "On Clicked (Button_Down)"() {is=title K2Node_ComponentBoundEvent}
   > MoveDown!(execute:Exec, self:Object Reference=Get EditorWidget(), Brush:Actor Object Reference=Get MyBrush()) -> (then:Exec) {is=title K2Node_Message}
   n1 = Event OnListItemObjectSet()
   > n2 = Cast To Actor!(execute, Object=@n1.ListItemObject) -> (then, CastFailed, AsActor:Actor Object Reference) {is=title K2Node_DynamicCast}
   > Set MyBrush(MyBrush:Actor Object Reference=@n2) -> (Output_Get:Actor Object Reference)
   > n3 = Cast To LandmassActor!(execute, Object=Get MyBrush()) -> (then, CastFailed, AsLandmass Actor:Landmass Actor Object Reference) {is=title K2Node_DynamicCast}
   > n4 = Cast To LandmassBrushManager!(execute, Object=Get BrushManager!(self:Landmass Actor Object Reference=@n3) -> (BrushManager:Landmass Manager Base Object Reference)) -> (then, CastFailed, AsLandmass Brush Manager:Landmass Brush Manager Object Reference) {is=title K2Node_DynamicCast}
   > Set EditorWidget(EditorWidget:EUW Landmass Editor Object Reference=Get EditorWidget!(self:Landmass Brush Manager Object Reference=@n4) -> (EditorWidget)) -> (Output_Get:EUW Landmass Editor Object Reference)
   n5 = Get MyBrush()
   > Set Text(Text="To Text (String)"(InString=GetDisplayName(Object=@n5)), self=Get TextBlock_87() -> (TextBlock_87:Text Object Reference))
   n6 = Event On Item Selection Changed!() -> (OutputDelegate, then, bIsSelected:bool) {is=title K2Node_Event}
   > Set Selected(Selected:bool=@n6.bIsSelected) -> (Output_Get:bool)
   > SetBrushColor(
    self=Get Background() -> (Background:Border Object Reference)
    InBrushColor=Select(
     Option 0:Linear Color Structure=Make Linear Color!(R:float=0.03, G:float=0.03, B:float=0.03, A:float=1) -> (LinearColor:Linear Color Structure) {is=title K2Node_MakeStruct}
     Option 1:Linear Color Structure=Make Linear Color(R=0.03, G=0.075, B=0.15, A=1)
     Index=@n6.bIsSelected
    ) -> (ReturnValue:Linear Color Structure)
   )
   Event On Item Expansion Changed!() -> (OutputDelegate, then, bIsExpanded:bool) {is=title K2Node_Event}
   > PrintString!(execute, InString=Item used in List changed. detected by List Widget Object, bPrintToScreen, bPrintToLog, TextColor, Duration, Key) -> (then)
   Event OnMouseEnter()
   > SetBrushColor(
    self=Get Background()
    InBrushColor=Select(
     Option 0=Make Linear Color(R=0.03, G=0.04, B=0.05, A=1)
     Option 1=Make Linear Color(R=0.03, G=0.1, B=0.22, A=1)
     Index=Get Selected() -> (Selected:bool)
    )
   )
   Event OnMouseLeave!() -> (OutputDelegate, then, MouseEvent:"Pointer Event Structure (by ref)") {is=title K2Node_Event}
   > SetBrushColor(
    self=Get Background()
    InBrushColor=Select(Option 0=Make Linear Color(R=0.03, G=0.03, B=0.03, A=1), Option 1=Make Linear Color(R=0.03, G=0.075, B=0.15, A=1), Index=Get Selected())
   )
   PrintString(InString=List Item Selected True)
   Event PreConstruct()
   Event Construct()
   Event Tick()
  ]}
 ListItemClicked {nodes [ListItemClicked() {is=title K2Node_FunctionEntry}]}
]
variables [
 EditorWidget {type=EUW_Landmass_Editor_C, default=null}
 ListItemClicked {}
 MyBrush {type=Actor, default=null}
 Selected {type=boolean}
]
widgets {
 info {inheritedWidgetCount=0, parentClass=&UserWidget, rootWidgetClass=&HorizontalBox, widgetCount=9}
 widgets [
  HorizontalBox_0 {widgetClassPath=&HorizontalBox, widget=HorizontalBox_0}
  Overlay_105 {
   widgetClassPath=&Overlay
   parent=HorizontalBox_0
   widget=Overlay_105
   slot {class=HorizontalBoxSlot, properties {size {sizeRule=Fill, value=1}}}
  }
  Background {
   widgetClassPath=&Border
   bIsVariable=true
   parent=Overlay_105
   widget=Background
   properties {brushColor=<rgba 0.03 0.03 0.03 1>, contentColorAndOpacity=<rgba 0.1771 0.1771 0.1771 1>}
   slot {class=OverlaySlot, properties {horizontalAlignment=HAlign_Fill, verticalAlignment=VAlign_Fill, padding=<ltrb 2 2 2 2>}}
  }
  HorizontalBox_142 {
   widgetClassPath=&HorizontalBox
   parent=Overlay_105
   widget=HorizontalBox_142
   slot {class=OverlaySlot, properties {horizontalAlignment=HAlign_Fill, padding=<ltrb 5 5 5 5>}}
  }
  TextBlock_87 {
   widgetClassPath=&TextBlock
   bIsVariable=true
   parent=HorizontalBox_142
   widget=TextBlock_87
   properties {
    text=Text Block
    colorAndOpacity {colorUseRule=UseColor_Specified, specifiedColor=<rgba 0.4219 0.4219 0.4219 1>}
    font {
     bForceMonospaced=false
     bMaterialIsStencil=false
     fontMaterial=null
     fontObject=&Roboto
     letterSpacing=0
     monospacedWidth=1
     size=12
     skewAmount=0
     typefaceFontName=Regular
     outlineSettings {
      bApplyOutlineToDropShadows=false
      bMiteredCorners=false
      bSeparateFillAlpha=false
      outlineMaterial=null
      outlineSize=0
      outlineColor=<rgba 0 0 0 1>
     }
    }
   }
   slot {class=HorizontalBoxSlot, properties {horizontalAlignment=HAlign_Center, verticalAlignment=VAlign_Center, padding=<ltrb 5 0 0 0>}}
  }
  Button_Up {
   widgetClassPath=&Button
   parent=HorizontalBox_142
   widget=Button_Up
   properties {widgetStyle {
     clickedSlateSound {resourceObject=null}
     disabled {
      bIsDynamicallyLoaded=false
      drawAs=NoDrawType
      imageType=NoImage
      mirroring=NoMirror
      resourceName=None
      resourceObject=null
      tiling=NoTile
      imageSize=<0 0>
      margin=<ltrb 0 0 0 0>
      outlineSettings {
       bUseBrushTransparency=false
       roundingType=HalfHeightRadius
       width=0
       color {colorUseRule=UseColor_Specified, specifiedColor=<rgba 0 0 0 0>}
       cornerRadii=<0 0 0 0>
      }
      tintColor {colorUseRule=UseColor_Specified, specifiedColor=<rgba 1 1 1 1>}
      uVRegion {bIsValid=false, max=<0 0>, min=<0 0>}
     }
     disabledForeground {colorUseRule=UseColor_Specified, specifiedColor=<rgba 0.5271 0.5271 0.5271 1>}
     hovered {
      bIsDynamicallyLoaded=false
      drawAs=RoundedBox
      imageType=NoImage
      mirroring=NoMirror
      resourceName=None
      resourceObject=null
      tiling=NoTile
      imageSize=<32 32>
      margin=<ltrb 0 0 0 0>
      outlineSettings {
       bUseBrushTransparency=true
       roundingType=FixedRadius
       width=1
       color {colorUseRule=UseColor_Specified, specifiedColor=<rgba 0.7243 0.7243 0.7243 1>}
       cornerRadii=<4 4 4 4>
      }
      tintColor {colorUseRule=UseColor_Specified, specifiedColor=<rgba 0.7243 0.7243 0.7243 1>}
      uVRegion {bIsValid=false, max=<0 0>, min=<0 0>}
     }
     hoveredForeground {colorUseRule=UseColor_Specified, specifiedColor=<rgba 0 1 1 1>}
     hoveredSlateSound {resourceObject=null}
     normal {
      bIsDynamicallyLoaded=false
      drawAs=RoundedBox
      imageType=NoImage
      mirroring=NoMirror
      resourceName=None
      resourceObject=null
      tiling=NoTile
      imageSize=<32 32>
      margin=<ltrb 0 0 0 0>
      outlineSettings {
       bUseBrushTransparency=true
       roundingType=FixedRadius
       width=1
       color {colorUseRule=UseColor_Specified, specifiedColor=<rgba 0.6951 0.6951 0.6951 1>}
       cornerRadii=<4 4 4 4>
      }
      tintColor {colorUseRule=UseColor_Specified, specifiedColor=<rgba 0.15 0.15 0.15 1>}
      uVRegion {bIsValid=false, max=<0 0>, min=<0 0>}
     }
     normalForeground {colorUseRule=UseColor_Specified, specifiedColor=<rgba 0 1 1 1>}
     normalPadding=<ltrb 5 5 5 5>
     pressed {
      bIsDynamicallyLoaded=false
      drawAs=RoundedBox
      imageType=NoImage
      mirroring=NoMirror
      resourceName=None
      resourceObject=null
      tiling=NoTile
      imageSize=<32 32>
      margin=<ltrb 0 0 0 0>
      outlineSettings {
       bUseBrushTransparency=true
       roundingType=FixedRadius
       width=1
       color {colorUseRule=UseColor_Specified, specifiedColor=<rgba 0.7243 0.7243 0.7243 1>}
       cornerRadii=<4 4 4 4>
      }
      tintColor {colorUseRule=UseColor_Specified, specifiedColor=<rgba 0.3843 0.3843 0.3843 1>}
      uVRegion {bIsValid=false, max=<0 0>, min=<0 0>}
     }
     pressedForeground {colorUseRule=UseColor_Specified, specifiedColor=<rgba 0 1 1 1>}
     pressedPadding=<ltrb 5 5 5 5>
     pressedSlateSound {resourceObject=null}
    }}
   slot {
    class=HorizontalBoxSlot
    properties {horizontalAlignment=HAlign_Right, verticalAlignment=VAlign_Center, padding=<ltrb 3 3 3 3>, size {sizeRule=Fill, value=1}}
   }
  }
  Image_44 {
   widgetClassPath=&Image
   bIsVariable=true
   parent=Button_Up
   widget=Image_44
   properties {brush {
     bIsDynamicallyLoaded=false
     drawAs=Image
     imageType=NoImage
     mirroring=NoMirror
     resourceName=None
     resourceObject=/Landmass/Landscape/ErosionBrush/Materials/Slate/CategoryArrow_Up
     tiling=NoTile
     imageSize=<16 16>
     margin=<ltrb 0 0 0 0>
     outlineSettings {
      bUseBrushTransparency=false
      roundingType=HalfHeightRadius
      width=0
      color {colorUseRule=UseColor_Specified, specifiedColor=<rgba 0 0 0 0>}
      cornerRadii=<0 0 0 0>
     }
     tintColor {colorUseRule=UseColor_Specified, specifiedColor=<rgba 1 1 1 1>}
     uVRegion {bIsValid=false, max=<0 0>, min=<0 0>}
    }}
   slot {class=ButtonSlot, properties {horizontalAlignment=HAlign_Left, verticalAlignment=VAlign_Top, padding=<ltrb 0 0 0 0>}}
  }
  Button_Down {
   widgetClassPath=&Button
   parent=HorizontalBox_142
   widget=Button_Down
   properties {widgetStyle {
     clickedSlateSound {resourceObject=null}
     disabled {
      bIsDynamicallyLoaded=false
      drawAs=NoDrawType
      imageType=NoImage
      mirroring=NoMirror
      resourceName=None
      resourceObject=null
      tiling=NoTile
      imageSize=<0 0>
      margin=<ltrb 0 0 0 0>
      outlineSettings {
       bUseBrushTransparency=false
       roundingType=HalfHeightRadius
       width=0
       color {colorUseRule=UseColor_Specified, specifiedColor=<rgba 0 0 0 0>}
       cornerRadii=<0 0 0 0>
      }
      tintColor {colorUseRule=UseColor_Specified, specifiedColor=<rgba 1 1 1 1>}
      uVRegion {bIsValid=false, max=<0 0>, min=<0 0>}
     }
     disabledForeground {colorUseRule=UseColor_Specified, specifiedColor=<rgba 0.5271 0.5271 0.5271 1>}
     hovered {
      bIsDynamicallyLoaded=false
      drawAs=RoundedBox
      imageType=NoImage
      mirroring=NoMirror
      resourceName=None
      resourceObject=null
      tiling=NoTile
      imageSize=<32 32>
      margin=<ltrb 0 0 0 0>
      outlineSettings {
       bUseBrushTransparency=true
       roundingType=FixedRadius
       width=1
       color {colorUseRule=UseColor_Specified, specifiedColor=<rgba 0.7243 0.7243 0.7243 1>}
       cornerRadii=<4 4 4 4>
      }
      tintColor {colorUseRule=UseColor_Specified, specifiedColor=<rgba 0.7243 0.7243 0.7243 1>}
      uVRegion {bIsValid=false, max=<0 0>, min=<0 0>}
     }
     hoveredForeground {colorUseRule=UseColor_Specified, specifiedColor=<rgba 0 1 1 1>}
     hoveredSlateSound {resourceObject=null}
     normal {
      bIsDynamicallyLoaded=false
      drawAs=RoundedBox
      imageType=NoImage
      mirroring=NoMirror
      resourceName=None
      resourceObject=null
      tiling=NoTile
      imageSize=<32 32>
      margin=<ltrb 0 0 0 0>
      outlineSettings {
       bUseBrushTransparency=true
       roundingType=FixedRadius
       width=1
       color {colorUseRule=UseColor_Specified, specifiedColor=<rgba 0.6951 0.6951 0.6951 1>}
       cornerRadii=<4 4 4 4>
      }
      tintColor {colorUseRule=UseColor_Specified, specifiedColor=<rgba 0.15 0.15 0.15 1>}
      uVRegion {bIsValid=false, max=<0 0>, min=<0 0>}
     }
     normalForeground {colorUseRule=UseColor_Specified, specifiedColor=<rgba 0 1 1 1>}
     normalPadding=<ltrb 5 5 5 5>
     pressed {
      bIsDynamicallyLoaded=false
      drawAs=RoundedBox
      imageType=NoImage
      mirroring=NoMirror
      resourceName=None
      resourceObject=null
      tiling=NoTile
      imageSize=<32 32>
      margin=<ltrb 0 0 0 0>
      outlineSettings {
       bUseBrushTransparency=true
       roundingType=FixedRadius
       width=1
       color {colorUseRule=UseColor_Specified, specifiedColor=<rgba 0.7243 0.7243 0.7243 1>}
       cornerRadii=<4 4 4 4>
      }
      tintColor {colorUseRule=UseColor_Specified, specifiedColor=<rgba 0.3843 0.3843 0.3843 1>}
      uVRegion {bIsValid=false, max=<0 0>, min=<0 0>}
     }
     pressedForeground {colorUseRule=UseColor_Specified, specifiedColor=<rgba 0 1 1 1>}
     pressedPadding=<ltrb 5 5 5 5>
     pressedSlateSound {resourceObject=null}
    }}
   slot {class=HorizontalBoxSlot, properties {horizontalAlignment=HAlign_Right, verticalAlignment=VAlign_Center, padding=<ltrb 3 3 3 3>}}
  }
  Image {
   widgetClassPath=&Image
   bIsVariable=true
   parent=Button_Down
   widget=Image
   properties {brush {
     bIsDynamicallyLoaded=false
     drawAs=Image
     imageType=NoImage
     mirroring=NoMirror
     resourceName=None
     resourceObject=/Landmass/Landscape/ErosionBrush/Materials/Slate/CategoryArrow_Rotated
     tiling=NoTile
     imageSize=<16 16>
     margin=<ltrb 0 0 0 0>
     outlineSettings {
      bUseBrushTransparency=false
      roundingType=HalfHeightRadius
      width=0
      color {colorUseRule=UseColor_Specified, specifiedColor=<rgba 0 0 0 0>}
      cornerRadii=<0 0 0 0>
     }
     tintColor {colorUseRule=UseColor_Specified, specifiedColor=<rgba 1 1 1 1>}
     uVRegion {bIsValid=false, max=<0 0>, min=<0 0>}
    }}
   slot {class=ButtonSlot, properties {horizontalAlignment=HAlign_Left, verticalAlignment=VAlign_Top, padding=<ltrb 0 0 0 0>}}
  }
 ]
}
