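# Editor utility widget that hosts a Landmass 3D preview viewport with editor-style mouse camera navigation.
# On construct it zeroes the viewport's default lighting, spawns a directional light, sky atmosphere and Landmass_WidgetComponents actor, then positions the camera.
# Input overrides store button and wheel state; Tick turns cursor deltas into rotate, pan and dolly moves.
# Set Camera LookAt frames a focus point; the Mouse Light Rotate sun-drag graph is not wired in.
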EditorUtilityWidgetBlueprint /Landmass/Landscape/BlueprintBrushes/Widgets/LandmassViewport_Widget
graphs [
 EventGraph {nodes [
   n1 = Set Camera LookAt!() -> (OutputDelegate, then, FocusPoint:Vector=~, Distance:double=~, ViewDirection:Vector=~) {is=title K2Node_CustomEvent}
   > SetViewLocation!(
    execute
    self:Viewport Object Reference=Get LandmassViewport() -> (LandmassViewport:Landmass Viewport Object Reference)
    Location:Vector=vector - vector(A=@n1.FocusPoint, B=vector * vector(A:double=@n1.Distance, B:Vector=@n1.ViewDirection) -> (ReturnValue=~))
   ) -> (then)
   > SetViewRotation!(
    execute
    self:Viewport Object Reference=Get LandmassViewport()
    Rotation:Rotator=Rotation From X Vector!(InVec:Vector=@n1.ViewDirection) -> (ReturnValue:Rotator)
   ) -> (then)
   Event Tick()
   > n2 = Sequence()
   > n3 = Branch(Condition=AND Boolean(A=Get LeftMouseDown() -> (LeftMouseDown:bool), B=Get RightMouseDown() -> (RightMouseDown:bool)))
   > K2Node_Composite_4 = Both Mouse Pan() {is=title K2Node_Composite}
   n4 = Branch(execute=@n3.else, Condition=Get LeftMouseDown())
   > K2Node_Composite_1 = L Mouse Rotate/Pan!(InputPin:Exec) {is=title K2Node_Composite}
   Branch(execute=@n4.else, Condition=Get RightMouseDown())
   > K2Node_Composite_2 = R Mouse: Rotate() {is=title K2Node_Composite}
   K2Node_Composite_3 = Mouse Wheel: Distance Pan(execute=@n2.then_1) {is=title K2Node_Composite}
   Event Construct()
   > n5 = Sequence()
   > SetEnableAdvancedFeatures!(execute, self:Viewport Object Reference=Get LandmassViewport(), InEnableAdvancedFeatures:bool=true) -> (then)
   > n6 = Spawn!(execute, self:Viewport Object Reference=Get LandmassViewport(), ActorClass:Actor Class Reference=/Script/Engine.DirectionalLight) -> (then, ReturnValue:Actor Object Reference)
   > n7 = Cast To DirectionalLight!(execute, Object=@n6) -> (then, CastFailed, AsDirectional Light:Directional Light Object Reference) {is=title K2Node_DynamicCast}
   > Set DirLight(DirLight:Directional Light Object Reference=@n7) -> (Output_Get:Directional Light Object Reference)
   > Spawn(self=Get LandmassViewport(), ActorClass=/Script/Engine.SkyAtmosphere)
   n8 = Get LandmassViewport()
   > SetLightIntensity!(execute, self:Viewport Object Reference=@n8, LightIntensity:float=0) -> (then)
   > SetSkyIntensity!(execute, self:Viewport Object Reference=@n8, LightIntensity:float=0) -> (then)
   > n9 = Spawn(self=Get LandmassViewport(), ActorClass=&/Landmass_WidgetComponents.Landmass_WidgetComponents_C)
   > n10 = Cast To Landmass_WidgetComponents!(execute, Object=@n9) -> (then, CastFailed, AsLandmass Widget Components:Landmass Widget Components Object Reference) {is=title K2Node_DynamicCast}
   > Set WidgetComponentBP(WidgetComponentBP:Landmass Widget Components Object Reference=@n10) -> (Output_Get:Landmass Widget Components Object Reference)
   n11 = Get LandmassViewport()
   SetViewLocation(execute=@n5.then_1, self=@n11, Location="-100,100,75")
   n12 = GetViewLocation!(self:Viewport Object Reference=@n11) -> (ReturnValue:Vector)
   > SetViewRotation(self=@n11, Rotation=Rotation From X Vector(InVec=vector * vector(A:Vector=@n12, B:float=-1) -> (ReturnValue=~)))
   Event OnMouseLeave!() -> (OutputDelegate, then, MouseEvent:"Pointer Event Structure (by ref)") {is=title K2Node_Event}
   > Set LeftMouseDown(LeftMouseDown:bool=false) -> (Output_Get:bool)
   > Set RightMouseDown(RightMouseDown:bool=false) -> (Output_Get:bool)
   K2Node_Composite_0 = Mouse Light Rotate!(InputPin:Exec) {is=title K2Node_Composite}
  ]}
 Mouse Light Rotate {
  path=EventGraph.K2Node_Composite_0.Mouse Light Rotate
  nodes [
   n1 = Get SunDirectionArrow!(self:Object Reference=Get Impostor Scene() -> (Impostor Scene:Object Reference)) -> (SunDirectionArrow:Static Mesh Component Object Reference)
   Set World Location!(execute, self:Scene Component Object Reference=@n1, NewLocation:Vector="0, 0, 0", bSweep:bool=false, bTeleport:bool=false) -> (then, SweepHitResult:Hit Result Structure) {is=title K2Node_CallFunction}
   n2 = Get Impostor Scene()
   > Set World Rotation!(
    execute
    self:Scene Component Object Reference=@n1
    NewRotation:Rotator=Get World Rotation!(self:Scene Component Object Reference=Get DirectionalLight!(self:Object Reference=@n2) -> (DirectionalLight:Directional Light Component Object Reference)) -> (ReturnValue:Rotator) {is=title K2Node_CallFunction}
    bSweep:bool=false
    bTeleport:bool=false
   ) -> (then, SweepHitResult:Hit Result Structure) {is=title K2Node_CallFunction}
   n3 = Get SunDirectionArrow(self=@n2)
   n4 = float + float(
    A=GetComponentBounds!(Component:Scene Component Object Reference=Get Capture_StaticMesh!(self:Object Reference=@n2) -> (Capture_StaticMesh:Static Mesh Component Object Reference)) -> (Origin:Vector, BoxExtent:Vector, SphereRadius:float).SphereRadius
    B=float * float(A=GetComponentBounds(Component=@n3).SphereRadius, B=2) {is=title K2Node_PromotableOperator}
   )
   > Add Local Offset!(
    execute
    self:Scene Component Object Reference=@n1
    DeltaLocation:Vector=MakeVector(X=float * float(A=@n4, B=-1), Y=0)
    bSweep:bool=false
    bTeleport:bool=false
   ) -> (then, SweepHitResult:Hit Result Structure)
   > SetVisibility!(
    execute
    self:Scene Component Object Reference=Get SunDirectionArrow(self=Get Impostor Scene())
    bNewVisibility:bool=true
    bPropagateToChildren:bool=false
   ) -> (then)
   Inputs!() -> (InputPin:Exec)
   > n5 = GetMousePositionOnViewport!(
    execute
    WorldContextObject:Object Reference=GetViewportWorld!(self:Viewport Object Reference=Get LandmassViewport()) -> (ReturnValue:World Object Reference)
   ) -> (then, ReturnValue:Vector 2D Structure)
   n6 = Get DirLight() -> (DirLight:Directional Light Object Reference)
   n7 = BreakRotator(InRot=Get World Rotation(self=Get LightComponent!(self:Light Object Reference=@n6) -> (LightComponent:Light Component Object Reference)))
   n8 = BreakVector2D(InVec=@n5)
   n9 = BreakVector2D(InVec=Get mouse_prev() -> (mouse_prev:Vector 2D Structure))
   n10 = float + float(A=@n7.Pitch, B=float / float(A=float - float(A=@n8.Y, B=@n9.Y), B=2))
   n11 = float - float(A=BreakVector2D(InVec=@n5), B=BreakVector2D(InVec=Get mouse_prev()))
   > Set World Rotation(
    self=Get LightComponent(self=Get DirLight())
    NewRotation=MakeRotator(Roll=@n7, Pitch="Clamp (Float)"(Value=@n10, Min=-89, Max=89), Yaw=float - float(A=@n7.Yaw, B=@n11))
    bSweep=false
    bTeleport=false
   )
   n12 = Get Impostor Scene()
   n13 = GetAxes!(A:Rotator=Get World Rotation(self=Get DirectionalLight(self=@n12))) -> (X:Vector, Y:Vector, Z:Vector)
   n14 = BreakVector(InVec=@n13)
   SetLowerHemisphereColor!(
    execute
    self:Sky Light Component Object Reference=Get SkyLight!(self:Object Reference=@n12) -> (SkyLight:Sky Light Component Object Reference)
    InLowerHemisphereColor:"Linear Color Structure (by ref)"=Select(
     Option 0:"Linear Color Structure (by ref)"=Make Linear Color!(R:float=0.05, G:float=0.05, B:float=0.075, A:float=1) -> (LinearColor:Linear Color Structure) {is=title K2Node_MakeStruct}
     Option 1:"Linear Color Structure (by ref)"=Make Linear Color(R=0, G=0, B=0, A=1)
     Index="float < float"(A=@n14.Z)
    ) -> (ReturnValue:"Linear Color Structure (by ref)")
   ) -> (then)
   "To String (Vector)"!(InVec:Vector=@n13) -> (ReturnValue:String)
   Outputs()
   Get DirectionalLight(self=Get Impostor Scene())
  ]
 }
 L Mouse Rotate/Pan {
  path=EventGraph.K2Node_Composite_1.L Mouse Rotate/Pan
  nodes [
   Inputs()
   > n1 = GetMousePositionOnViewport(WorldContextObject=GetViewportWorld(self=Get LandmassViewport()))
   n2 = BreakRotator(InRot=GetViewRotation!(self:Viewport Object Reference=Get LandmassViewport()) -> (ReturnValue:Rotator))
   n3 = Get mouse_prev()
   n4 = float / float(A=float - float(A=BreakVector2D(InVec=@n1), B=BreakVector2D(InVec=@n3)), B=2)
   > SetViewRotation(self=Get LandmassViewport(), Rotation=MakeRotator(Roll=@n2, Pitch=@n2.Pitch, Yaw=float + float(A=@n2.Yaw, B=@n4)))
   > n5 = GetMousePositionOnViewport(WorldContextObject=GetViewportWorld(self=Get LandmassViewport()))
   n6 = Get LandmassViewport()
   n7 = Get LandmassViewport()
   n8 = GetViewRotation(self=@n7)
   n9 = Normalize!(
    A:Vector=vector * vector(A=BreakRotIntoAxes!(InRot:"Rotator (by ref)"=@n8) -> (X:Vector, Y:Vector, Z:Vector), B:Vector="1,1,0") -> (ReturnValue=~)
    Tolerance:float=0.0001
   ) -> (ReturnValue:Vector) {is=title K2Node_CallFunction}
   n10 = BreakVector2D(InVec=@n5)
   n11 = BreakVector2D(InVec=Get mouse_prev())
   n12 = float * float(A=float * float(A=float - float(A=@n10.Y, B=@n11.Y), B=-1), B=Get Zoom Sensitivity() -> (Zoom Sensitivity:double))
   > SetViewLocation(self=@n6, Location=vector + vector(A=GetViewLocation(self=@n7), B=vector * vector(A=@n9, B:double=@n12) -> (ReturnValue=~)))
   SetViewRotation(self=@n6, Rotation=Reroute Node(InputPin:Rotator) -> (OutputPin:Rotator))
   Outputs()
  ]
 }
 "R Mouse: Rotate" {
  path=EventGraph.K2Node_Composite_2.R Mouse: Rotate
  nodes [
   Inputs!() -> (execute:Exec)
   > n1 = GetMousePositionOnViewport(WorldContextObject=GetViewportWorld(self=Get LandmassViewport()))
   n2 = BreakRotator(InRot=GetViewRotation(self=Get LandmassViewport()))
   n3 = Get mouse_prev()
   n4 = float / float(A=float - float(A=BreakVector2D(InVec=@n1), B=BreakVector2D(InVec=@n3)), B=2)
   > SetViewRotation(self=Get LandmassViewport(), Rotation=MakeRotator(Roll=@n2, Pitch=@n2.Pitch, Yaw=float + float(A=@n2.Yaw, B=@n4)))
   > n5 = GetMousePositionOnViewport(WorldContextObject=GetViewportWorld(self=Get LandmassViewport()))
   n6 = BreakRotator(InRot=GetViewRotation(self=Get LandmassViewport()))
   n7 = Get mouse_prev()
   n8 = float / float(A=float - float(A=BreakVector2D(InVec=@n7).Y, B=BreakVector2D(InVec=@n5).Y), B=2)
   > SetViewRotation(self=Get LandmassViewport(), Rotation=MakeRotator(Roll=@n6, Pitch=float + float(A=@n6.Pitch, B=@n8), Yaw=@n6.Yaw))
   Outputs()
  ]
 }
 "Mouse Wheel: Distance Pan" {
  path=EventGraph.K2Node_Composite_3.Mouse Wheel: Distance Pan
  nodes [
   Inputs()
   > n1 = GetMousePositionOnViewport(WorldContextObject=GetViewportWorld(self=Get LandmassViewport()))
   n2 = Get LandmassViewport()
   n3 = Get Rotation X Vector!(InRot:Rotator=GetViewRotation(self=@n2)) -> (ReturnValue:Vector) {is=title K2Node_CallFunction}
   n4 = float * float(A=float * float(A=Get MouseWheelDelta() -> (MouseWheelDelta:double), B=20), B=Get Zoom Sensitivity())
   > SetViewLocation(self=Get LandmassViewport(), Location=vector + vector(A=GetViewLocation(self=@n2), B=vector * vector(A=@n3, B=@n4) -> (ReturnValue=~)))
   > n5 = GetMousePositionOnViewport(WorldContextObject=GetViewportWorld(self=Get LandmassViewport()))
   > Set mouse_prev(mouse_prev:Vector 2D Structure=@n5) -> (Output_Get:Vector 2D Structure)
   > Set MouseWheelDelta(MouseWheelDelta:double=0) -> (Output_Get:double)
   vector - vector(A:double=BreakVector2D(InVec=Get mouse_prev()).Y, B:double=BreakVector2D(InVec=@n1).Y)
   Outputs()
  ]
 }
 Both Mouse Pan {
  path=EventGraph.K2Node_Composite_4.Both Mouse Pan
  nodes [
   Inputs()
   > n1 = GetMousePositionOnViewport(WorldContextObject=GetViewportWorld(self=Get LandmassViewport()))
   n2 = Get LandmassViewport()
   n3 = GetAxes(A=GetViewRotation(self=@n2))
   n4 = Get mouse_prev()
   n5 = float * float(A=float - float(A=BreakVector2D(InVec=@n1), B=BreakVector2D(InVec=@n4)), B=Get Zoom Sensitivity())
   > SetViewLocation(self=Get LandmassViewport(), Location=vector + vector(A=GetViewLocation(self=@n2), B=vector * vector(A=@n3.Y, B=@n5) -> (ReturnValue=~)))
   > n6 = GetMousePositionOnViewport(WorldContextObject=GetViewportWorld(self=Get LandmassViewport()))
   n7 = Get LandmassViewport()
   n8 = Get mouse_prev()
   n9 = float * float(A=float - float(A=BreakVector2D(InVec=@n8).Y, B=BreakVector2D(InVec=@n6).Y), B=Get Zoom Sensitivity())
   > SetViewLocation(self=Get LandmassViewport(), Location=vector + vector(A=GetViewLocation(self=@n7), B=vector * vector(A="0,0,1", B=@n9) -> (ReturnValue=~)))
   GetAxes(A=GetViewRotation(self=@n7))
   Outputs()
  ]
 }
 OnKeyDown {nodes [
   n1 = OnKeyDown!() -> (then, MyGeometry:Geometry Structure, InKeyEvent:Key Event Structure) {is=title K2Node_FunctionEntry}
   > n2 = Branch(Condition="Equal (Key)"(A:Key Structure=GetKey!(Input:"Key Event Structure (by ref)"=@n1.InKeyEvent) -> (ReturnValue:Key Structure), B:Key Structure=L) -> (ReturnValue:bool=~) {is=title K2Node_PromotableOperator})
   > n3 = Set L_Key_Pressed(L_Key_Pressed:bool=true) -> (Output_Get:bool)
   n4 = Branch(execute=[@n6.then, @n3.then], Condition="Equal (Key)"(A=GetKey(Input=@n1.InKeyEvent), B=LeftAlt) -> (ReturnValue=~))
   > n5 = Set Alt Pressed(Alt Pressed:bool=true) -> (Output_Get:bool)
   Return Node!(
    execute=[@n4.else, @n5.then]
    ReturnValue:Event Reply Structure=Make Event Reply!() -> (EventReply:Event Reply Structure) {is=title K2Node_MakeStruct}
   )
   n6 = Set L_Key_Pressed(execute=@n2.else, L_Key_Pressed=false)
  ]}
 OnKeyUp {nodes [
   n1 = OnKeyUp!() -> (then, MyGeometry:Geometry Structure, InKeyEvent:Key Event Structure) {is=title K2Node_FunctionEntry}
   > n2 = Branch(Condition="Equal (Key)"(A=GetKey(Input=@n1.InKeyEvent), B=L) -> (ReturnValue=~))
   > n3 = Set L_Key_Pressed(L_Key_Pressed=false)
   n4 = Branch(execute=[@n6.then, @n3.then], Condition="Equal (Key)"(A=GetKey(Input=@n1.InKeyEvent), B=LeftAlt) -> (ReturnValue=~))
   > n5 = Set Alt Pressed(Alt Pressed=false)
   Return Node(execute=[@n4.else, @n5.then], ReturnValue=Make Event Reply())
   n6 = Set L_Key_Pressed(execute=@n2.else, L_Key_Pressed=false)
  ]}
 OnMouseButtonDown {nodes [
   n1 = OnMouseButtonDown!() -> (then, MyGeometry:Geometry Structure, MouseEvent:"Pointer Event Structure (by ref)") {is=title K2Node_FunctionEntry}
   > Set LeftMouseDown(LeftMouseDown=Is Mouse Button Down!(Input:"Pointer Event Structure (by ref)"=@n1.MouseEvent, MouseButton:Key Structure=LeftMouseButton) -> (ReturnValue:bool))
   > n2 = Set RightMouseDown(RightMouseDown=Is Mouse Button Down(Input=@n1.MouseEvent, MouseButton=RightMouseButton))
   > n3 = GetMousePositionOnViewport(WorldContextObject=GetViewportWorld(self=Get LandmassViewport()))
   > Set mouse_prev(mouse_prev=@n3)
   > Return Node(ReturnValue=Make Event Reply())
   PrintString(InString=Append(A="RMB: ", B="To String (Boolean)"(InBool=@n2)))
  ]}
 OnMouseButtonUp {nodes [
   n1 = OnMouseButtonUp!() -> (then, MyGeometry:Geometry Structure, MouseEvent:"Pointer Event Structure (by ref)") {is=title K2Node_FunctionEntry}
   n2 = Get Effecting Button!(Input:"Pointer Event Structure (by ref)"=@n1.MouseEvent) -> (ReturnValue:Key Structure) {is=title K2Node_CallFunction}
   > n3 = Branch(Condition="Equal (Key)"(A=@n2, B=LeftMouseButton) -> (ReturnValue=~))
   > n4 = Set LeftMouseDown(LeftMouseDown=false)
   n5 = Branch(execute=[@n3.else, @n4.then], Condition="Equal (Key)"(A=@n2, B=RightMouseButton) -> (ReturnValue=~))
   > n6 = Set RightMouseDown(RightMouseDown=false)
   Return Node(execute=[@n5.else, @n6.then], ReturnValue=Make Event Reply())
   Get LeftMouseDown()
  ]}
 OnMouseMove {nodes [
   OnMouseMove!() -> (then, MyGeometry:Geometry Structure, MouseEvent:"Pointer Event Structure (by ref)") {is=title K2Node_FunctionEntry}
   > Branch(Condition=AND Boolean(A=Get LeftMouseDown(), B=Get RightMouseDown()))
   n1 = Make Event Reply()
   > Return Node(ReturnValue=@n1)
   LockMouse!(Reply:"Event Reply Structure (by ref)"=@n1, CapturingWidget:Widget Object Reference) -> (ReturnValue:Event Reply Structure)
   SetMousePosition!(Reply:"Event Reply Structure (by ref)"=@n1, NewMousePosition:Vector 2D Structure) -> (ReturnValue:Event Reply Structure)
  ]}
 OnMouseWheel {nodes [
   n1 = OnMouseWheel!() -> (then, MyGeometry:Geometry Structure, MouseEvent:"Pointer Event Structure (by ref)") {is=title K2Node_FunctionEntry}
   > Set MouseWheelDelta(MouseWheelDelta=Get Wheel Delta!(Input:"Pointer Event Structure (by ref)"=@n1.MouseEvent) -> (ReturnValue:float) {is=title K2Node_CallFunction})
   > Return Node(ReturnValue=Make Event Reply())
  ]}
]
variables [
 Alt Pressed {type=boolean}
 CamearFocusPosition {type=Vector}
 CurrentPivot {type=Vector}
 DirLight {type=DirectionalLight, default=null}
 Dragging {type=boolean}
 FloorMesh {type=boolean}
 L_Key_Pressed {type=boolean}
 LeftMouseDown {type=boolean}
 mouse_prev {type=Vector2D, default=<0 0>}
 MouseWheelDelta {type=number}
 RightMouseDown {type=boolean}
 WidgetComponentBP {type=Landmass_WidgetComponents_C, default=null}
 Zoom Sensitivity {type=number, default=200}
]
widgets {
 info {rootWidgetClass=&Overlay, widgetCount=2}
 widgets [
  Overlay_62 {widgetClassPath=&Overlay, widget=Overlay_62}
  LandmassViewport {
   widgetClassPath=&/LandmassViewport
   bIsVariable=true
   parent=Overlay_62
   widget=LandmassViewport
   properties {visibility=SelfHitTestInvisible}
   slot {class=OverlaySlot, properties {horizontalAlignment=HAlign_Fill, verticalAlignment=VAlign_Fill}}
  }
 ]
}
